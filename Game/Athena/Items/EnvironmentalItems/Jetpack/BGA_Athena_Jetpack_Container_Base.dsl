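; Base Blueprint for a searchable, destructible jetpack container that grants loot from a tier group.
; On BeginPlay the server rolls replicated HasLoot from a spawn-chance scalable float; its OnRep hides the mesh and disables collision, interaction and marking when false.
; Interaction clears HasLoot, plays a searched cue, then gives loot directly if the player's item count is under five, otherwise drops it.
; Death plays a destroyed cue, drops remaining loot and hides the mesh.
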(event EventBeginPlay
  ;[0.000818 1 0] BeginPlay Spawn Chance
  (if (Networking|IsServer)
    (Variables|Default|SetHasLoot? (Math|Random|RandomBoolwithWeight (ScalableFloat|GetValueatLevel (Variables|Default|GetRow_SpawnChance))))))

(event Interaction|EventBlueprintOnInteract
  (params (InteractingPawn :type Fort%20Pawn%20Object%20Reference) (InteractionBeingAttempted :type EInteractionBeingAttempted%20Enum :default "FirstInteraction"))
  ;[0 0.07046 1] Interaction and Loot Logic
  (Variables|Default|SetHasLoot? false)
  (Ability|ExecuteGameplayCueWithParamsReplicated (Variables|Default|GetGC_JetpackContainer_Searched) (Ability|GameplayCue|MakeGameplayCueParameters :NormalizedMagnitude 0.0 :RawMagnitude 0.0 :Location (Transformation|GetWorldLocation (Variables|Default|GetJetpackMesh)) :Normal "0,0,0" :GameplayEffectLevel 1 :AbilityLevel 1 :bReplicateLocationWhenUsingMinimalRepProxy false))
  ;;[1 1 1] Set the last pawn to interact and look for open inventory
  (bind _output_get (Variables|Default|SetLastInteractingPawn InteractingPawn))
  (bind _returnvalue (Default|HasOpenInventorySlot _output_get))
  (if _returnvalue
    ;;[1 1 1] If open give them the item
    (CallFunction|GiveItemDirectly)
    (else
      ;;[1 1 1] If not open, put it on the floor
      (CallFunction|SpawnLoot))))

(event Custom|GiveItemDirectly
  ;[0 0.07046 1] Interaction and Loot Logic
  ;;[1 1 1] Get the last pawn to interact and cast to them
  (Development|PrintString "GivingItem")
  (bind _asfort_player_controller (Utilities|Casting|CastToFortPlayerController (Pawn|GetController (Variables|Default|GetLastInteractingPawn))))
  ;;[1 1 1] Get the item we're giving and cast to it
  (Default|GetItemtoGrant (Variables|Default|GetLootTierGroupName) _asfort_player_controller))

(event Custom|SpawnLoot
  ;[0 0.07046 1] Interaction and Loot Logic
  ;;[1 1 1] Useful once loot table is setup
  (Class|AthenaFunctionLibrary|SpawnLootfromTable (Variables|Default|GetLootTierGroupName) 1 -1 (Transformation|GetActorLocation) "0, 0, 0" 0 false true))

(event EventOnDeathServer
  (params (Damage :type Float%20%28single-precision%29) (DamageTags :type Gameplay%20Tag%20Container%20Structure%20%28by%20ref%29) (Momentum :type Vector :default "0, 0, 0") (HitInfo :type Hit%20Result%20Structure%20%28by%20ref%29) (InstigatedBy :type Controller%20Object%20Reference) (DamageCauser :type Actor%20Object%20Reference) (EffectContext :type Gameplay%20Effect%20Context%20Handle%20Structure))
  ;[1 0.008232 0] Destroyed and Death Logic
  (Ability|ExecuteGameplayCueReplicated (Variables|Default|GetGC_JetpackContainer_Destroyed))
  (if (Variables|Default|GetHasLoot?)
    (CallFunction|SpawnLoot)))

(event EventOnDeathPlayEffects
  (params (Damage :type Float%20%28single-precision%29) (DamageTags :type Gameplay%20Tag%20Container%20Structure%20%28by%20ref%29) (Momentum :type Vector :default "0, 0, 0") (HitInfo :type Hit%20Result%20Structure%20%28by%20ref%29) (InstigatedBy :type Fort%20Pawn%20Object%20Reference) (DamageCauser :type Actor%20Object%20Reference) (EffectContext :type Gameplay%20Effect%20Context%20Handle%20Structure))
  ;[1 0.008232 0] Destroyed and Death Logic
  (Development|SetHiddeninGame (Variables|Default|GetJetpackMesh) true true))

(event OnLinkedActorDestroyed(FortLinkToActor)
  (params (DamageCauser :type Actor%20Object%20Reference))
  ;[1 0.008232 0] Destroyed and Death Logic
  (Actor|DestroyActor))

(fn BlueprintCanInteract
  (params (InteractingPawn :type Fort%20Pawn%20Object%20Reference) (InteractionBeingAttempted :type EInteractionBeingAttempted%20Enum :default "FirstInteraction") (InteractionType :type TInteractionType%20Enum :default "IT_NoInteraction"))
  (return true))

(fn BlueprintGetInteractionString
  (params (InteractingPawn :type Fort%20Pawn%20Object%20Reference) (InteractionBeingAttempted :type EInteractionBeingAttempted%20Enum :default "FirstInteraction"))
  (return (Variables|Default|GetInteractMessage)))

(fn BlueprintGetInteractionTime
  (params (InteractingPawn :type Fort%20Pawn%20Object%20Reference) (InteractionBeingAttempted :type EInteractionBeingAttempted%20Enum :default "FirstInteraction"))
  (return (ScalableFloat|GetValueatLevel (Variables|Default|GetRow_InteractionTime)) true))

(fn GetItemToGrant
  (params (TierGroupName :type Name) (PlayerControllerRef :type Fort%20Player%20Controller%20Object%20Reference))
  (bind (_outloottodrop _returnvalue) (Loot|PickLootDrops TierGroupName 0 -1 "(GameplayTags=)"))
  (if _returnvalue
    (for _array_element _outloottodrop
      (bind (_array_element_itemdefinition _array_element_itemguid _array_element_count) (Class|FortItemFunctionLibrary|BreakFortItemEntry _array_element))
      (bind _asfort_world_item_definition (Utilities|Casting|CastToFortWorldItemDefinition _array_element_itemdefinition))
      (Inventory|GiveItemtoInventoryOwner :InventoryOwner PlayerControllerRef :ItemDefinition _asfort_world_item_definition :NumberToGive _array_element_count :bNotifyPlayer true :ItemLevel -1 :PickupInstigatorHandle -1 :bUseItemPickupAnalyticEvent false :WeaponAmmoOverride -1))))

(fn HasOpenInventorySlot
  (params (Target :type Pawn%20Object%20Reference))
  (bind _asfort_player_controller (Utilities|Casting|CastToFortPlayerController (Pawn|GetController Target))
    (:then
      (bind _returnvalue (IndexSelectionModel|GetNumItems(Message) _asfort_player_controller))
      (Development|PrintString (Utilities|String|ToString(Integer) _returnvalue))
      (return (< _returnvalue 5)))
    (:CastFailed
      (return false))))

(fn OnRep_HasLoot?
  (bind _jetpackmesh (Variables|Default|GetJetpackMesh))
  (if (not (Variables|Default|GetHasLoot?))
    (Collision|SetCollisionEnabled _jetpackmesh "NoCollision")
    (Development|SetHiddeninGame _jetpackmesh true true)
    (Variables|Interaction|SetbAllowInteract false)
    (MarkableActor|SetCanBeMarked false)))

(fn ConstructionScript)
